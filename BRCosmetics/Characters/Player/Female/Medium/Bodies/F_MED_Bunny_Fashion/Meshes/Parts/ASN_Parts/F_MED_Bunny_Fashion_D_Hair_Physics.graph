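# Physics asset for the hair of the F_MED_Bunny_Fashion female character body, covering the bangs.
# Capsule bodies form three chains of three bones (center, left, right), with the first bone of each constrained to a capsule on the head.
# Side chains share uniformly tight swing and twist limits, while the center chain swings more freely, most widely at its root.
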PhysicsAsset /BRCosmetics/Characters/Player/Female/Medium/Bodies/F_MED_Bunny_Fashion/Meshes/Parts/ASN_Parts/F_MED_Bunny_Fashion_D_Hair_Physics
bodies {
 dyn_bang_c_01 [{shapeType=Capsule, length=0.505, radius=0.505}]
 dyn_bang_c_02 [{shapeType=Capsule, length=2.505, radius=1.255, center=<2 0 0>, rotation=<rot 90 0 0>}]
 dyn_bang_c_03 [{shapeType=Capsule, length=2.1492, radius=1.2289, center=<2 0 0>, rotation=<rot 90 0 0>}]
 dyn_bang_l_01 [{shapeType=Capsule, length=3.005, radius=1.505, center=<3 0 0>, rotation=<rot 90 0 0>}]
 dyn_bang_l_02 [{shapeType=Capsule, length=6.005, radius=1.505, center=<3 0 0>, rotation=<rot 90 0 0>}]
 dyn_bang_l_03 [{shapeType=Capsule, length=4.505, radius=1.255, center=<3 0 0>, rotation=<rot 90 0 0>}]
 dyn_bang_r_01 [{shapeType=Capsule, length=3.255, radius=2.005, center=<3 0 0>, rotation=<rot 90 0 0>}]
 dyn_bang_r_02 [{shapeType=Capsule, length=4.3594, radius=1.5813, center=<3 0 0>, rotation=<rot 90 0 0>}]
 dyn_bang_r_03 [{shapeType=Capsule, length=4.289, radius=1.1577, center=<2.5278 0 0>, rotation=<rot 90 0 0>}]
 head [{shapeType=Capsule, length=4.9387, radius=6.2388, rotation=<rot -90 0.0033 -0.0032>}]
}
constraints [
 {bone1Name=dyn_bang_c_01, bone2Name=head, swing1LimitDegrees=45, swing2LimitDegrees=45, twistLimitDegrees=15}
 {bone1Name=dyn_bang_c_02, bone2Name=dyn_bang_c_01, swing1LimitDegrees=10, swing2LimitDegrees=10, twistLimitDegrees=5}
 {bone1Name=dyn_bang_c_03, bone2Name=dyn_bang_c_02, swing1LimitDegrees=25, swing2LimitDegrees=15, twistLimitDegrees=5}
 {bone1Name=dyn_bang_l_01, bone2Name=head, swing1LimitDegrees=5, swing2LimitDegrees=3, twistLimitDegrees=3}
 {bone1Name=dyn_bang_l_02, bone2Name=dyn_bang_l_01, swing1LimitDegrees=5, swing2LimitDegrees=3, twistLimitDegrees=3}
 {bone1Name=dyn_bang_l_03, bone2Name=dyn_bang_l_02, swing1LimitDegrees=5, swing2LimitDegrees=3, twistLimitDegrees=3}
 {bone1Name=dyn_bang_r_01, bone2Name=head, swing1LimitDegrees=5, swing2LimitDegrees=3, twistLimitDegrees=3}
 {bone1Name=dyn_bang_r_02, bone2Name=dyn_bang_r_01, swing1LimitDegrees=5, swing2LimitDegrees=3, twistLimitDegrees=3}
 {bone1Name=dyn_bang_r_03, bone2Name=dyn_bang_r_02, swing1LimitDegrees=5, swing2LimitDegrees=3, twistLimitDegrees=3}
]
