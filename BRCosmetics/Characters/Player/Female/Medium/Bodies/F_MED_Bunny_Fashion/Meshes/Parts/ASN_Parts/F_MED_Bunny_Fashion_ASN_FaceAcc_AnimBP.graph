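# Animation Blueprint for the Bunny Fashion character's face accessory mesh, posing its hair bang and hood bones on top of the attached parent's pose.
# Cosmetic variant meta tags, checked when customization loads, set hair_base_active, choosing shrunken, offset bones when true or neck-relative RigidBody simulation when false.
# Both branches then constrain the bang tips between neck and head and scale them.
# Head rotation is computed each update but drives no node.
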AnimBlueprint /BRCosmetics/Characters/Player/Female/Medium/Bodies/F_MED_Bunny_Fashion/Meshes/Parts/ASN_Parts/F_MED_Bunny_Fashion_ASN_FaceAcc_AnimBP
target_skeleton=&/F_MED_Bunny_Fashion_ASN_FaceAcc_Skeleton
graphs [
 AnimGraph {nodes [
   "Transform (Modify) Bone - Bone: dyn_bang_l_02"(Alpha=1, Translation="-1,0,0", Rotation="-10,0.000,0.000", Scale="0.7,0.7,0.7") {
    rotationMode=BMM_Additive
    scaleMode=BMM_Additive
    translationMode=BMM_Additive
    alphaBoolBlend {blendOption=Linear}
    boneToModify {boneName=dyn_bang_l_02}
    rotation {pitch=-10}
    scale=<0.7 0.7 0.7>
    translation {x=-1}
   }
   n1 = RigidBody(ComponentPose=Local To Component(LocalPose=Copy Pose From Mesh() {bUseAttachedParent=true})) {
    bClampLinearTranslationLimitToRefPose=true
    bOverrideWorldGravity=true
    overridePhysicsAsset=&/F_MED_Bunny_Fashion_D_Hair_Physics
    simulationSpace=BaseBoneSpace
    alphaBoolBlend {blendOption=Linear}
    baseBoneRef {boneName=neck_01}
    componentAppliedLinearAccClamp=<700 700 700>
    componentLinearAccScale {y=0.3, z=0.3}
    overrideWorldGravity {z=-980}
   }
   n2 = Component To Local(ComponentPose=@n1)
   n3 = "Transform (Modify) Bone - Bone: dyn_bang_l_01"(
    ComponentPose=Local To Component(LocalPose=Copy Pose From Mesh() {bUseAttachedParent=true})
    Alpha=1
    Translation="0,0,0"
    Rotation="0,0,0"
    Scale="0.7,0.7,0.7"
   ) {
    rotationMode=BMM_Additive
    scaleMode=BMM_Additive
    translationMode=BMM_Additive
    alphaBoolBlend {blendOption=Linear}
    boneToModify {boneName=dyn_bang_l_01}
    scale=<0.7 0.7 0.7>
   }
   n4 = "Transform (Modify) Bone - Bone: dyn_hood_l_04"(ComponentPose=@n3, Alpha=1, Translation="-1,0,0", Rotation="-10,0.000,0.000", Scale="0.6,0.6,0.6") {
    rotationMode=BMM_Additive
    scaleMode=BMM_Additive
    translationMode=BMM_Additive
    alphaBoolBlend {blendOption=Linear}
    boneToModify {boneName=dyn_hood_l_04}
    rotation {pitch=-10}
    scale=<0.6 0.6 0.6>
    translation {x=-1}
   }
   n5 = "Transform (Modify) Bone - Bone: dyn_hood_l_05"(ComponentPose=@n4, Alpha=1, Translation="-1,0,0", Rotation="0,0,0", Scale="0.5,0.5,0.5") {
    rotationMode=BMM_Additive
    scaleMode=BMM_Additive
    translationMode=BMM_Additive
    alphaBoolBlend {blendOption=Linear}
    boneToModify {boneName=dyn_hood_l_05}
    scale=<0.5 0.5 0.5>
    translation {x=-1}
   }
   n6 = Component To Local(ComponentPose=@n5)
   n7 = Get hair_base_active() -> (hair_base_active:bool)
   n8 = "Constraint - dyn_bang_l_03 (PP)"!(
    ComponentPose=Local To Component(LocalPose=Blend Poses by bool(BlendPose_0=@n6, BlendPose_1=@n2, bActiveValue=@n7) {blendType=Linear})
    Alpha=0.7
    ConstraintWeights_0=0.5
    ConstraintWeights_1:float=0.5
   ) -> (Pose) {
    alphaBoolBlend {blendOption=Linear}
    boneToModify {boneName=dyn_bang_l_03}
    constraintSetup [{targetBone {boneName=neck_01}}, {targetBone {boneName=head}}]
    constraintWeights [1,1]
   }
   n9 = "Constraint - dyn_bang_r_03 (PP)"!(ComponentPose=@n8, Alpha=0.5, ConstraintWeights_0=0.5, ConstraintWeights_1:float=0.5) -> (Pose) {
    alphaBoolBlend {blendOption=Linear}
    boneToModify {boneName=dyn_bang_r_03}
    constraintSetup [{targetBone {boneName=neck_01}}, {targetBone {boneName=head}}]
    constraintWeights [1,1]
   }
   n10 = "Transform (Modify) Bone - Bone: dyn_bang_l_03"(ComponentPose=@n9, Alpha=1, Translation="0,0,0", Rotation="0,0,0", Scale="0.6,0.6,0.6") {
    rotationMode=BMM_Additive
    scaleMode=BMM_Additive
    alphaBoolBlend {blendOption=Linear}
    boneToModify {boneName=dyn_bang_l_03}
    scale=<0.6 0.6 0.6>
   }
   n11 = "Transform (Modify) Bone - Bone: dyn_bang_r_03"(ComponentPose=@n10, Alpha=1, Translation="0,0,0", Rotation="10,0.000,0.000", Scale="0.5,0.5,0.5") {
    rotationMode=BMM_Additive
    scaleMode=BMM_Additive
    alphaBoolBlend {blendOption=Linear}
    boneToModify {boneName=dyn_bang_r_03}
    rotation {pitch=10}
    scale=<0.5 0.5 0.5>
   }
   Output Pose(Result=Component To Local(ComponentPose=@n11)) {layerGroup=None}
   Get head Rot x() -> (head Rot x:double)
  ]}
 EventGraph {nodes [
   Event BlueprintUpdateAnimation()
   n1 = BreakTransform(InTransform=GetDeltaTransformFromRefPose(self=GetOwnerSkeletalMeshComponent(), BoneName=head, BaseName=neck_01))
   > Set head Rot x(head Rot x:double=MapRangeClamped(Value=BreakRotator(InRot=@n1.Rotation), InRangeA=0, InRangeB=35)) -> (Output_Get:double)
   Event OnCharacterCustomizationLoaded()
   > n2 = Branch(Condition=CharacterItemDefinitionHasAnyMetaTag(Tags=Make Array!("[0]"=(TagName="Cosmetics.Variant.Property.Stage1"), "[1]":Gameplay Tag Structure=(TagName="Cosmetics.Variant.Property")) -> (Array)))
   > Set hair_base_active(hair_base_active:bool=true) -> (Output_Get:bool)
   Set hair_base_active(execute=@n2.else, hair_base_active=false)
   Event BlueprintInitializeAnimation()
   > Set Mesh_To_Copy(Mesh_To_Copy:Skeletal Mesh Component Object Reference=GetOwnerSkeletalMeshComponent()) -> (Output_Get:Skeletal Mesh Component Object Reference)
   TryGetPawnOwner()
   TryGetPawnOwner()
  ]}
]
variables [
 hair_base_active {type=boolean}
 head Rot x {type=number}
 Mesh_To_Copy {type=SkeletalMeshComponent}
]
